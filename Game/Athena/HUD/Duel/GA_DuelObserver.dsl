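; Gameplay ability that observes duels between player pawns and relays their state for the Athena duel HUD overlay.
; On event activation, the server casts the event's instigator and target to FortPlayerPawns and registers them through AddDuel.
; When a duel changes, it passes the new state and overlay data to UpdateDuel on the ability owner's Athena player controller.
; That handler is labeled server-side but checks only that the controller is valid.
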(event Ability|EventActivateAbilityFromEvent
  (params (EventData :type Gameplay%20Event%20Data%20Structure%20%28by%20ref%29))
  (bind (_eventtag _instigator _target) (Utilities|Struct|BreakGameplayEventData EventData))
  ;[0 0.003112 0.02604] SERVER-SIDE
  ;;[0 0 0] Initial setup from gameplay event
  (Utilities|IsServer
    (:true
      (CallFunction|AddDuel (Utilities|Casting|CastToFortPlayerPawn _instigator) (Utilities|Casting|CastToFortPlayerPawn _target)))))

(event EventOnDuelChange
  (params (NewDuelState :type EDuelState%20Enum :default "Started") (ChangedDuelOverlayData :type Duel%20Overlay%20Data%20Structure%20%28by%20ref%29))
  (bind _returnvalue (Ability|GetActorInfo))
  (bind (_owneractor _avataractor _playercontroller) (Utilities|Struct|BreakGameplayAbilityActorInfo _returnvalue))
  ;[0 0.003112 0.02604] SERVER-SIDE
  ;;[0 0 0] Handling duel change
  (Utilities|IsValid _playercontroller
    (:Is%20Valid
      (bind _asfort_player_controller_athena (Utilities|Casting|CastToFortPlayerControllerAthena _playercontroller))
      (Class|FortPlayerControllerAthena|UpdateDuel _asfort_player_controller_athena NewDuelState ChangedDuelOverlayData))))
