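# HUD widget for duels against NPCs, showing animated start and end banners with portrait, title and subtitle.
# Setup picks Intro for Started or DuelElim for Won and Lost, reveals the reward only on Won, and tints its glow from the rarity palette.
# Showing resets fades and plays the chosen animation; outro runs FadeAnim through a simple-animations panel.
# DuelElim switches to the eliminated portrait, animating shake, material-driven reveal and starburst effects.
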WidgetBlueprint /Game/Athena/HUD/Duel/AthenaDuelOverlay
animations [
 Intro {
  binding_count=7
  length_seconds=3.25
  object=&self:Intro
  sequence {bindings [
    Image_PlayerPulse {
     id=#1
     possessed_class=Image
     tracks [{
       type=MovieScene2DTransformTrack
       sections [{
         type=MovieScene2DTransformSection
         channels [
          Translation.X {type=MovieSceneScriptingFloatChannel}
          Translation.Y {type=MovieSceneScriptingFloatChannel}
          Angle {type=MovieSceneScriptingFloatChannel, keys [[0,0],[18.288,180]]}
          Scale.X {type=MovieSceneScriptingFloatChannel}
          Scale.Y {type=MovieSceneScriptingFloatChannel}
          Shear.X {type=MovieSceneScriptingFloatChannel}
          Shear.Y {type=MovieSceneScriptingFloatChannel}
         ]
        }]
      }]
    }
    Image_ChallengedNPCPortrait {
     id=#2
     possessed_class=Image
     tracks [{
       type=MovieSceneWidgetMaterialTrack
       sections [{
         type=MovieSceneParameterSection
         channels [
          Mask Rotation {type=MovieSceneScriptingFloatChannel, keys [[0,-0.75],[18.288,-0.375]]}
          CustomUV {type=MovieSceneScriptingFloatChannel, keys [[0,-0.5938],[18.288,1.029]]}
          PortraitFillR {type=MovieSceneScriptingFloatChannel, const [0,1]}
          PortraitFillG {type=MovieSceneScriptingFloatChannel, const [0,1]}
          PortraitFillB {type=MovieSceneScriptingFloatChannel, const [0,1]}
          PortraitFillA {type=MovieSceneScriptingFloatChannel, keys [[17.3187,1],[21.174,0]]}
         ]
        }]
      }]
    }
    CommonWidgetSwitcher_IconType {
     id=#3
     possessed_class=CommonWidgetSwitcherLegacy
     tracks [{
       type=MovieSceneIntegerTrack
       display_name=Active Widget Index
       sections [{type=MovieSceneIntegerSection, end_frame=0, start_frame=0, channels [None {type=MovieSceneScriptingIntegerChannel, const [0,1]}]}]
      }]
    }
    HorizontalBox_1 {
     id=#4
     possessed_class=HorizontalBox
     tracks [
      {
       type=MovieScene2DTransformTrack
       sections [{
         type=MovieScene2DTransformSection
         channels [
          Translation.X {type=MovieSceneScriptingFloatChannel}
          Translation.Y {type=MovieSceneScriptingFloatChannel}
          Angle {type=MovieSceneScriptingFloatChannel}
          Scale.X {type=MovieSceneScriptingFloatChannel, keys [[0,2.7136],[18.041,3],[24.876,1],[25.732,1.1149],[28.876,1]]}
          Scale.Y {type=MovieSceneScriptingFloatChannel, keys [[0,2.7136],[18.041,3],[24.876,1],[25.732,1.1149],[28.876,1]]}
          Shear.X {type=MovieSceneScriptingFloatChannel}
          Shear.Y {type=MovieSceneScriptingFloatChannel}
         ]
        }]
      }
      {
       type=MovieSceneFloatTrack
       display_name=Render Opacity
       sections [{type=MovieSceneFloatSection, channels [None {type=MovieSceneScriptingFloatChannel, keys [[0,0],[16.199,1],[49,1]]}]}]
      }
     ]
    }
    RichText_SubTitle {
     id=#5
     possessed_class=CommonRichTextBlock
     tracks [{
       type=MovieSceneWidgetMaterialTrack
       display_name=Default Text Style Override.Font.Outline Settings.OutlineMaterial
       sections [{type=MovieSceneParameterSection, channels [TimeOffset {type=MovieSceneScriptingFloatChannel, keys [[37,0.5754],[65,-1]]}]}]
      }]
    }
    CommonBorder_0 {
     id=#6
     possessed_class=CommonBorder
     tracks [
      {
       type=MovieSceneFloatVectorTrack
       display_name=Desired Size Scale
       sections [{
         type=MovieSceneFloatVectorSection
         channels [
          Vector.X {type=MovieSceneScriptingFloatChannel, keys [[31.9413,0],[36.9413,1]]}
          Vector.Y {type=MovieSceneScriptingFloatChannel}
         ]
        }]
      }
      {
       type=MovieSceneByteTrack
       display_name=Clipping
       sections [{type=MovieSceneByteSection, channels [None {type=MovieSceneScriptingByteChannel, keys [[0,1],[36.9413,0]]}]}]
      }
      {
       type=MovieSceneFloatTrack
       display_name=Render Opacity
       sections [{type=MovieSceneFloatSection, channels [None {type=MovieSceneScriptingFloatChannel, keys [[31.9413,0],[32.6133,1]]}]}]
      }
     ]
    }
    Overlay_Container {
     id=#7
     possessed_class=Overlay
     tracks [{
       type=MovieSceneFloatTrack
       display_name=Render Opacity
       sections [{type=MovieSceneFloatSection, channels [None {type=MovieSceneScriptingFloatChannel, const [1,1]}]}]
      }]
    }
   ]}
 }
 DuelElim {
  binding_count=8
  length_seconds=3.5271
  object=&self:DuelElim
  sequence {bindings [
    CommonBorder_0 {
     id=#8
     possessed_class=CommonBorder
     tracks [
      {
       type=MovieSceneByteTrack
       display_name=Clipping
       sections [{type=MovieSceneByteSection, channels [None {type=MovieSceneScriptingByteChannel, keys [[0,1],[52.567,0]]}]}]
      }
      {
       type=MovieSceneFloatVectorTrack
       display_name=Desired Size Scale
       sections [{
         type=MovieSceneFloatVectorSection
         channels [
          Vector.X {type=MovieSceneScriptingFloatChannel, keys [[47.567,0],[52.567,1]]}
          Vector.Y {type=MovieSceneScriptingFloatChannel}
         ]
        }]
      }
      {
       type=MovieSceneFloatTrack
       display_name=Render Opacity
       sections [{type=MovieSceneFloatSection, channels [None {type=MovieSceneScriptingFloatChannel, keys [[47.567,0],[48.239,1]]}]}]
      }
     ]
    }
    Image_shards {
     id=#9
     possessed_class=Image
     tracks [
      {
       type=MovieSceneFloatTrack
       display_name=Render Opacity
       sections [{type=MovieSceneFloatSection, channels [None {type=MovieSceneScriptingFloatChannel, keys [[22.935,0],[31.8147,1],[45.688,0]]}]}]
      }
      {
       type=MovieSceneWidgetMaterialTrack
       sections [{
         type=MovieSceneParameterSection
         channels [
          ManualTime {type=MovieSceneScriptingFloatChannel, keys [[22.994,-0.3512],[45.688,0.3056]]}
          MaskScaleR {type=MovieSceneScriptingFloatChannel, keys [[24.2677,0.2],[36.141,0.5975],[43.141,1.1483]]}
          MaskScaleG {type=MovieSceneScriptingFloatChannel, keys [[24.2677,0.2],[43.141,1.1483]]}
          MaskScaleB {type=MovieSceneScriptingFloatChannel, const [0,1]}
          MaskScaleA {type=MovieSceneScriptingFloatChannel, keys [[24.2677,1.171],[45.708,0.0661]]}
         ]
        }]
      }
      {
       type=MovieScene2DTransformTrack
       sections [{
         type=MovieScene2DTransformSection
         channels [
          Translation.X {type=MovieSceneScriptingFloatChannel}
          Translation.Y {type=MovieSceneScriptingFloatChannel}
          Angle {type=MovieSceneScriptingFloatChannel}
          Scale.X {type=MovieSceneScriptingFloatChannel, keys [[22.994,0],[45.688,0.879]]}
          Scale.Y {type=MovieSceneScriptingFloatChannel, keys [[22.994,0],[45.688,0.879]]}
          Shear.X {type=MovieSceneScriptingFloatChannel}
          Shear.Y {type=MovieSceneScriptingFloatChannel}
         ]
        }]
      }
     ]
    }
    Image_ElimPortrait {
     id=#10
     possessed_class=Image
     tracks [
      {
       type=MovieSceneWidgetMaterialTrack
       sections [{
         type=MovieSceneParameterSection
         channels [
          InnerMaskRevealRadius {type=MovieSceneScriptingFloatChannel, keys [[13.6297,0],[14.312,0.0791],[15.9313,0.79],[50.0947,1.1952]]}
          InnerMaskRevealDensity {type=MovieSceneScriptingFloatChannel, keys [[24.1107,0.1488],[43.461,0.6705],[50.0903,1.3043]]}
          MaskRevealDensity {type=MovieSceneScriptingFloatChannel, keys [[23,6.7401],[24,-2.605],[50.0947,-1.4592]]}
          MaskRevealRadius {type=MovieSceneScriptingFloatChannel, keys [[23,6.5759],[24,2.0746],[50.0947,0.0831]]}
          GlobalScale {type=MovieSceneScriptingFloatChannel, keys [[14.285,0.7286],[14.63,0.6186],[15.806,0.7286],[23,0.7286],[24,0.6581],[35.688,1.0294]]}
          RevealDecal1R {type=MovieSceneScriptingFloatChannel, keys [[4.7387,0.169],[5.9383,0.0081]]}
          RevealDecal1G {type=MovieSceneScriptingFloatChannel, keys [[4.7387,0.3929],[5.9383,0.37]]}
          RevealDecal1B {type=MovieSceneScriptingFloatChannel, const [1,2]}
          RevealDecal1A {type=MovieSceneScriptingFloatChannel, keys [[4.7387,0],[5.9383,0.3048]]}
          RevealDecal2R {type=MovieSceneScriptingFloatChannel, keys [[9.5383,0.7395],[10.7383,0.8914]]}
          RevealDecal2G {type=MovieSceneScriptingFloatChannel, keys [[9.5383,0.7204],[10.7383,0.7957]]}
          RevealDecal2B {type=MovieSceneScriptingFloatChannel, const [1,2]}
          RevealDecal2A {type=MovieSceneScriptingFloatChannel, keys [[9.5383,0],[10.7383,0.3238]]}
          RevealDecal3R {type=MovieSceneScriptingFloatChannel, const [0.4762,2]}
          RevealDecal3G {type=MovieSceneScriptingFloatChannel, keys [[7.1383,0.349],[8.3383,0.2476]]}
          RevealDecal3B {type=MovieSceneScriptingFloatChannel, const [1,2]}
          RevealDecal3A {type=MovieSceneScriptingFloatChannel, keys [[7.1383,0],[8.3383,0.3524]]}
         ]
        }]
      }
      {
       type=MovieScene2DTransformTrack
       sections [{
         type=MovieScene2DTransformSection
         channels [
          Translation.X {
           type=MovieSceneScriptingFloatChannel
           keys [[4.74,0.8599],[4.881,3.5826],[5.0187,-0.7219],[5.2013,2.1187],[5.385,3.5826],[5.523,-0.7219],[5.7053,2.1187],[5.9617,0.8599],[7.1383,0.8599],[7.2793,3.5826],[7.417,-0.7219],[7.5997,2.1187],[7.7833,3.5826],[7.9213,-0.7219],[8.1037,2.1187],[8.36,0.8599],[9.5383,0.8599],[9.6793,3.5826],[9.817,-0.7219],[9.9997,2.1187],[10.1833,3.5826],[10.3213,-0.7219],[10.5037,2.1187],[10.76,0.8599],[14.312,0.8599],[14.429,3.5826],[14.5437,-0.7219],[14.696,2.1187],[14.849,3.5826],[14.964,-0.7219],[15.116,2.1187],[15.3297,0.8599],[22.994,0.8599],[23.111,3.5826],[23.2257,-0.7219],[23.378,2.1187],[23.531,3.5826],[23.646,-0.7219],[23.798,2.1187],[24.0117,0.8599]]
          }
          Translation.Y {
           type=MovieSceneScriptingFloatChannel
           keys [[4.74,-2.3837],[4.8717,-6.573],[5.0417,-1.3573],[5.2323,-4.3897],[5.3757,-6.573],[5.546,-1.3573],[5.7367,-4.3897],[5.9617,-2.3837],[7.1383,-2.3837],[7.27,-6.573],[7.44,-1.3573],[7.6307,-4.3897],[7.774,-6.573],[7.9443,-1.3573],[8.135,-4.3897],[8.36,-2.3837],[9.5383,-2.3837],[9.67,-6.573],[9.84,-1.3573],[10.0307,-4.3897],[10.174,-6.573],[10.3443,-1.3573],[10.535,-4.3897],[10.76,-2.3837],[14.312,-2.3837],[14.4213,-6.573],[14.563,-1.3573],[14.7217,-4.3897],[14.8413,-6.573],[14.9833,-1.3573],[15.142,-4.3897],[15.3297,-2.3837],[22.994,-2.3837],[23.1033,-6.573],[23.245,-1.3573],[23.4037,-4.3897],[23.5233,-6.573],[23.6653,-1.3573],[23.824,-4.3897],[24.0117,-2.3837]]
          }
          Angle {type=MovieSceneScriptingFloatChannel}
          Scale.X {type=MovieSceneScriptingFloatChannel}
          Scale.Y {type=MovieSceneScriptingFloatChannel}
          Shear.X {type=MovieSceneScriptingFloatChannel}
          Shear.Y {type=MovieSceneScriptingFloatChannel}
         ]
        }]
      }
     ]
    }
    CommonWidgetSwitcher_IconType {
     id=#11
     possessed_class=CommonWidgetSwitcherLegacy
     tracks [{
       type=MovieSceneIntegerTrack
       display_name=Active Widget Index
       sections [{type=MovieSceneIntegerSection, channels [None {type=MovieSceneScriptingIntegerChannel, keys [[0,1],[0.171,2]]}]}]
      }]
    }
    Image_Reward {
     id=#12
     possessed_class=Image
     tracks [{
       type=MovieSceneFloatTrack
       display_name=Render Opacity
       sections [{type=MovieSceneFloatSection, channels [None {type=MovieSceneScriptingFloatChannel, keys [[46.3803,0],[52.372,1]]}]}]
      }]
    }
    Image_Glow {
     id=#13
     possessed_class=Image
     tracks [
      {
       type=MovieSceneColorTrack
       display_name=Color and Opacity
       sections [{
         type=MovieSceneColorSection
         channels [
          Color.R {type=MovieSceneScriptingFloatChannel, const [1,2]}
          Color.G {type=MovieSceneScriptingFloatChannel, keys [[29.623,0.1531],[52.567,1]]}
          Color.B {type=MovieSceneScriptingFloatChannel, keys [[29.623,0.2228],[52.567,1]]}
          Color.A {type=MovieSceneScriptingFloatChannel, const [1,2]}
         ]
        }]
      }
      {
       type=MovieSceneFloatTrack
       display_name=Render Opacity
       sections [{type=MovieSceneFloatSection, channels [None {type=MovieSceneScriptingFloatChannel, keys [[22.935,0],[29.623,1],[52.567,1],[70.542,0.22]]}]}]
      }
     ]
    }
    Image_sparks {
     id=#14
     possessed_class=Image
     tracks [
      {
       type=MovieSceneFloatTrack
       display_name=Render Opacity
       sections [{type=MovieSceneFloatSection, channels [None {type=MovieSceneScriptingFloatChannel, keys [[22.935,0],[25.5897,1],[33.463,0]]}]}]
      }
      {
       type=MovieSceneWidgetMaterialTrack
       sections [{type=MovieSceneParameterSection, channels [ManualTime {type=MovieSceneScriptingFloatChannel, keys [[22.994,-2.129],[33.463,-1.062]]}]}]
      }
      {
       type=MovieScene2DTransformTrack
       sections [{
         type=MovieScene2DTransformSection
         channels [
          Translation.X {type=MovieSceneScriptingFloatChannel}
          Translation.Y {type=MovieSceneScriptingFloatChannel}
          Angle {type=MovieSceneScriptingFloatChannel}
          Scale.X {type=MovieSceneScriptingFloatChannel, keys [[22.688,0.587],[33.688,1.1556]]}
          Scale.Y {type=MovieSceneScriptingFloatChannel, keys [[22.688,0.587],[33.688,1.1556]]}
          Shear.X {type=MovieSceneScriptingFloatChannel}
          Shear.Y {type=MovieSceneScriptingFloatChannel}
         ]
        }]
      }
     ]
    }
    HorizontalBox_1 {
     id=#15
     possessed_class=HorizontalBox
     tracks [
      {
       type=MovieSceneFloatTrack
       display_name=Render Opacity
       sections [{type=MovieSceneFloatSection, channels [None {type=MovieSceneScriptingFloatChannel, keys [[0,0],[3.4393,1],[49,1]]}]}]
      }
      {
       type=MovieScene2DTransformTrack
       sections [{
         type=MovieScene2DTransformSection
         channels [
          Translation.X {type=MovieSceneScriptingFloatChannel}
          Translation.Y {type=MovieSceneScriptingFloatChannel}
          Angle {type=MovieSceneScriptingFloatChannel}
          Scale.X {type=MovieSceneScriptingFloatChannel, keys [[0,2.7136],[42,2.7136],[46.9227,1]]}
          Scale.Y {type=MovieSceneScriptingFloatChannel, keys [[0,2.7136],[42,2.7136],[46.9227,1]]}
          Shear.X {type=MovieSceneScriptingFloatChannel}
          Shear.Y {type=MovieSceneScriptingFloatChannel}
         ]
        }]
      }
     ]
    }
   ]}
 }
]
graphs [EventGraph {nodes [
   Event OnOutroOverlay() {is=title K2Node_Event}
   > PlayAnimationByName!(
    execute
    self:Fort Simple Widget Animations Panel Object Reference=Get SimpleAnimations() -> (SimpleAnimations:Fort Simple Widget Animations Panel Object Reference)
    Name:"Name (by ref)"=Get FadeAnim() -> (FadeAnim:Name=None)
    Speed:float=1
    StartAtTime:float=0
   ) -> (then, ReturnValue:bool)
   n1 = Event OnSetupForActivation!() -> (OutputDelegate, then, DuelState:EDuelState Enum=Started, RarityColorPalette:"Fort Color Palette Structure (by ref)") {is=title K2Node_Event}
   n2 = Get DuelElim() -> (DuelElim:Widget Animation Object Reference)
   > Set AnimationToPlayOnShow(AnimationToPlayOnShow:Widget Animation Object Reference=Select!(
     Started:Widget Animation Object Reference=Get Intro()
     Won:Widget Animation Object Reference=@n2
     Lost:Widget Animation Object Reference=@n2
     Index:EDuelState Enum=@n1.DuelState
    ) -> (ReturnValue:Widget Animation Object Reference)) -> (Output_Get:Widget Animation Object Reference)
   > SetVisibility(
    self=Get Overlay_Reward() -> (Overlay_Reward:Overlay Object Reference)
    InVisibility=Select(
     Started:ESlateVisibility Enum=Collapsed
     Won:ESlateVisibility Enum=HitTestInvisible
     Lost:ESlateVisibility Enum=Collapsed
     Index=@n1.DuelState
    ) -> (ReturnValue:ESlateVisibility Enum)
   )
   > n3 = GetDynamicMaterial(self=Get Image_Glow() -> (Image_Glow:Image Object Reference))
   n4 = Break Fort Color Palette!(FortColorPalette:"Fort Color Palette Structure (by ref)"=@n1.RarityColorPalette) -> (
    Color1:Linear Color Structure
    Color2:Linear Color Structure
    Color3:Linear Color Structure
    Color4:Linear Color Structure
    Color5:Linear Color Structure
   ) {is=title K2Node_BreakStruct}
   > SetVectorParameterValue(self=@n3, ParameterName=Color1, Value=@n4)
   > SetVectorParameterValue(self=@n3, ParameterName=Color2, Value=@n4.Color2)
   Event OnShowOverlay() {is=title K2Node_Event}
   > StopAllAnimations(self:User Widget Object Reference)
   > StopAnimationByName!(
    execute
    self:Fort Simple Widget Animations Panel Object Reference=Get SimpleAnimations()
    Name:"Name (by ref)"=Get FadeAnim() -> (FadeAnim=None)
   ) -> (then, ReturnValue:bool)
   > SetRenderOpacity(self=Get Overlay_Container() -> (Overlay_Container:Overlay Object Reference), InOpacity=1)
   n5 = Get AnimationToPlayOnShow() -> (AnimationToPlayOnShow:Widget Animation Object Reference)
   > Is Valid(InputObject=@n5)
   > PlayAnimation(InAnimation=@n5)
  ]}]
variables [
 AnimationToPlayOnShow {type=WidgetAnimation, default=null}
 FadeAnim {type=string, default=FadeAnim}
]
widgets {
 info {inheritedWidgetCount=6, parentClass=/Script/FortniteUI.AthenaDuelOverlay, rootWidgetClass=&SizeBox, widgetCount=23}
 widgets [
  SizeBox_0 {
   widgetClassPath=&SizeBox
   widget=SizeBox_0
   properties {bOverride_MinDesiredHeight=true, bOverride_MinDesiredWidth=true, minDesiredHeight=200, minDesiredWidth=600}
  }
  SimpleAnimations {
   widgetClassPath=/Script/FortniteUI.FortSimpleWidgetAnimationsPanel
   bIsVariable=true
   parent=SizeBox_0
   widget=SimpleAnimations
   properties {animations [
     IntroAnim {
      bIsEnabled=true
      onAnimationsFinished=()
      targets [{
        alpha=/Game/Athena/HUD/Discoverability/SpecialEventStarted_IntroAnim_RenderOpacity
        bResetOnFinish=false
        bShouldAnimateAlpha=true
        bShouldAnimateScale=false
        bShouldAnimateTranslation=false
        scale=null
        targetName=Overlay_Container
        translation=/Game/Athena/HUD/Discoverability/SpecialEventStarted_IntroAnim_Translation
        materialScalarAnimations []
       }]
     }
     FadeAnim {
      bIsEnabled=true
      onAnimationsFinished=()
      targets [{
        alpha=/Game/Athena/HUD/Discoverability/SpecialEventStarted_FadeAnim_RenderOpacity
        bResetOnFinish=false
        bShouldAnimateAlpha=true
        bShouldAnimateScale=false
        bShouldAnimateTranslation=false
        scale=null
        targetName=Overlay_Container
        translation=/Game/Athena/HUD/Discoverability/SpecialEventStarted_FadeAnim_Translation
        materialScalarAnimations []
       }]
     }
    ]}
   slot {class=SizeBoxSlot, properties {horizontalAlignment=HAlign_Center, verticalAlignment=VAlign_Top}}
  }
  Overlay_Container {
   widgetClassPath=&Overlay
   bIsVariable=true
   parent=SimpleAnimations
   widget=Overlay_Container
   slot {class=FortSimpleWidgetAnimationsSlot}
  }
  MainBox {
   widgetClassPath=&VerticalBox
   bIsVariable=true
   parent=Overlay_Container
   widget=MainBox
   properties {renderTransformPivot=<0.5 0>}
   slot {class=OverlaySlot, properties {horizontalAlignment=HAlign_Fill, verticalAlignment=VAlign_Fill}}
  }
  HorizontalBox_1 {
   widgetClassPath=&HorizontalBox
   parent=MainBox
   widget=HorizontalBox_1
   properties {renderTransformPivot=<0.5 0.3>}
   slot {class=VerticalBoxSlot, properties {horizontalAlignment=HAlign_Center, padding=<ltrb 0 0 0 -10>}}
  }
  CommonWidgetSwitcher_IconType {
   widgetClassPath=&CommonWidgetSwitcherLegacy
   bIsVariable=true
   parent=HorizontalBox_1
   widget=CommonWidgetSwitcher_IconType
   properties {activeWidgetIndex=1, transitionDuration=0, renderTransform {angle=0, scale=<1 1>, shear=<0 0>, translation=<19 0>}}
   slot {class=HorizontalBoxSlot, properties {horizontalAlignment=HAlign_Left, verticalAlignment=VAlign_Center, padding=<ltrb 0 0 32 0>}}
  }
  Overlay_PawnIcon {
   widgetClassPath=&Overlay
   bIsVariable=true
   parent=CommonWidgetSwitcher_IconType
   widget=Overlay_PawnIcon
   slot {class=WidgetSwitcherSlot, properties {horizontalAlignment=HAlign_Right, verticalAlignment=VAlign_Center, padding=<ltrb 0 -64 0 -64>}}
  }
  Image_PlayerPulse {
   widgetClassPath=&Image
   bIsVariable=true
   parent=Overlay_PawnIcon
   widget=Image_PlayerPulse
   properties {
    visibility=HitTestInvisible
    brush {
     bIsDynamicallyLoaded=false
     drawAs=Image
     imageType=NoImage
     mirroring=NoMirror
     resourceName=None
     resourceObject=/Game/Athena/HUD/LTM/Infiltration/Materials/M_UI_GameplayDiamondPing_Pulse_Add_Inst
     tiling=NoTile
     imageSize=<32 32>
     margin=<ltrb 0 0 0 0>
     outlineSettings {
      bUseBrushTransparency=false
      roundingType=HalfHeightRadius
      width=0
      color {colorUseRule=UseColor_Specified, specifiedColor=<rgba 0 0 0 0>}
      cornerRadii=<0 0 0 0>
     }
     tintColor {colorUseRule=UseColor_Specified, specifiedColor=<rgba 1 1 1 1>}
     uVRegion {bIsValid=false, max=<0 0>, min=<0 0>}
    }
    colorAndOpacity=<rgba 1 0.0169 0 1>
   }
   slot {class=OverlaySlot, properties {horizontalAlignment=HAlign_Fill, verticalAlignment=VAlign_Fill, padding=<ltrb -15 -15 -15 -15>}}
  }
  Image_ChallengedNPCPortrait {
   widgetClassPath=&Image
   bInherited=true
   bIsVariable=true
   parent=Overlay_PawnIcon
   widget=Image_ChallengedNPCPortrait
   properties {
    visibility=SelfHitTestInvisible
    brush {
     bIsDynamicallyLoaded=false
     drawAs=Image
     imageType=NoImage
     mirroring=NoMirror
     resourceName=None
     resourceObject=/Game/UI/SocialPanel/Art/Materials/UI_UserPortrait_Enemy
     tiling=NoTile
     imageSize=<128 128>
     margin=<ltrb 0 0 0 0>
     outlineSettings {
      bUseBrushTransparency=false
      roundingType=HalfHeightRadius
      width=0
      color {colorUseRule=UseColor_Specified, specifiedColor=<rgba 0 0 0 0>}
      cornerRadii=<0 0 0 0>
     }
     tintColor {colorUseRule=UseColor_Specified, specifiedColor=<rgba 1 1 1 1>}
     uVRegion {bIsValid=false, max=<0 0>, min=<0 0>}
    }
   }
   slot {class=OverlaySlot, properties {horizontalAlignment=HAlign_Center, verticalAlignment=VAlign_Center}}
  }
  Image_Elim {
   widgetClassPath=&Image
   bIsVariable=true
   parent=Overlay_PawnIcon
   widget=Image_Elim
   properties {
    visibility=HitTestInvisible
    brush {
     bIsDynamicallyLoaded=false
     drawAs=Image
     imageType=NoImage
     mirroring=NoMirror
     resourceName=None
     resourceObject=/Game/Athena/Items/Accolades/Materials/M_UI_Icon_Elim
     tiling=NoTile
     imageSize=<75 75>
     margin=<ltrb 0 0 0 0>
     outlineSettings {
      bUseBrushTransparency=false
      roundingType=HalfHeightRadius
      width=0
      color {colorUseRule=UseColor_Specified, specifiedColor=<rgba 0 0 0 0>}
      cornerRadii=<0 0 0 0>
     }
     tintColor {colorUseRule=UseColor_Specified, specifiedColor=<rgba 1 1 1 1>}
     uVRegion {bIsValid=false, max=<0 0>, min=<0 0>}
    }
    renderTransform {angle=-14, scale=<1 1>, shear=<0 0>, translation=<-23 -52>}
   }
   slot {class=OverlaySlot, properties {horizontalAlignment=HAlign_Center, verticalAlignment=VAlign_Bottom, padding=<ltrb -64 -64 -64 -64>}}
  }
  Overlay_EliminatedPawnIcon {
   widgetClassPath=&Overlay
   parent=CommonWidgetSwitcher_IconType
   widget=Overlay_EliminatedPawnIcon
   slot {class=WidgetSwitcherSlot}
  }
  Image_ElimPortrait {
   widgetClassPath=&Image
   bInherited=true
   bIsVariable=true
   parent=Overlay_EliminatedPawnIcon
   widget=Image_ElimPortrait
   properties {brush {
     bIsDynamicallyLoaded=false
     drawAs=Image
     imageType=NoImage
     mirroring=NoMirror
     resourceName=None
     resourceObject=&UI_UserPortrait_Enemy
     tiling=NoTile
     imageSize=<128 128>
     margin=<ltrb 0 0 0 0>
     outlineSettings {
      bUseBrushTransparency=false
      roundingType=HalfHeightRadius
      width=0
      color {colorUseRule=UseColor_Specified, specifiedColor=<rgba 0 0 0 0>}
      cornerRadii=<0 0 0 0>
     }
     tintColor {colorUseRule=UseColor_Specified, specifiedColor=<rgba 1 1 1 1>}
     uVRegion {bIsValid=false, max=<0 0>, min=<0 0>}
    }}
   slot {class=OverlaySlot, properties {horizontalAlignment=HAlign_Fill, verticalAlignment=VAlign_Fill}}
  }
  Image_sparks {
   widgetClassPath=&Image
   bIsVariable=true
   parent=Overlay_EliminatedPawnIcon
   widget=Image_sparks
   properties {
    renderOpacity=0
    visibility=SelfHitTestInvisible
    brush {
     bIsDynamicallyLoaded=false
     drawAs=Image
     imageType=NoImage
     mirroring=NoMirror
     resourceName=None
     resourceObject=/Game/Athena/UI/InGame/Interrogation/Materials/M_UI_Starburst_Inst3
     tiling=NoTile
     imageSize=<256 256>
     margin=<ltrb 0 0 0 0>
     outlineSettings {
      bUseBrushTransparency=false
      roundingType=HalfHeightRadius
      width=0
      color {colorUseRule=UseColor_Specified, specifiedColor=<rgba 0 0 0 0>}
      cornerRadii=<0 0 0 0>
     }
     tintColor {colorUseRule=UseColor_Specified, specifiedColor=<rgba 1 1 1 1>}
     uVRegion {bIsValid=false, max=<0 0>, min=<0 0>}
    }
   }
   slot {class=OverlaySlot, properties {horizontalAlignment=HAlign_Center, verticalAlignment=VAlign_Center, padding=<ltrb -128 -128 -128 -128>}}
  }
  Image_shards {
   widgetClassPath=&Image
   bIsVariable=true
   parent=Overlay_EliminatedPawnIcon
   widget=Image_shards
   properties {
    renderOpacity=0
    visibility=SelfHitTestInvisible
    brush {
     bIsDynamicallyLoaded=false
     drawAs=Image
     imageType=NoImage
     mirroring=NoMirror
     resourceName=None
     resourceObject=/Game/Athena/UI/InGame/Interrogation/Materials/M_UI_Starburst_Triangle
     tiling=NoTile
     imageSize=<200 200>
     margin=<ltrb 0 0 0 0>
     outlineSettings {
      bUseBrushTransparency=false
      roundingType=HalfHeightRadius
      width=0
      color {colorUseRule=UseColor_Specified, specifiedColor=<rgba 0 0 0 0>}
      cornerRadii=<0 0 0 0>
     }
     tintColor {colorUseRule=UseColor_Specified, specifiedColor=<rgba 1 1 1 1>}
     uVRegion {bIsValid=false, max=<0 0>, min=<0 0>}
    }
   }
   slot {class=OverlaySlot, properties {horizontalAlignment=HAlign_Center, verticalAlignment=VAlign_Center, padding=<ltrb -256 -256 -256 -256>}}
  }
  Overlay_Reward {
   widgetClassPath=&Overlay
   bIsVariable=true
   parent=Overlay_EliminatedPawnIcon
   widget=Overlay_Reward
   slot {class=OverlaySlot, properties {horizontalAlignment=HAlign_Center, verticalAlignment=VAlign_Center, padding=<ltrb -128 -128 -128 -128>}}
  }
  Image_Reward {
   widgetClassPath=&Image
   bInherited=true
   bIsVariable=true
   parent=Overlay_Reward
   widget=Image_Reward
   properties {brush {
     bIsDynamicallyLoaded=false
     drawAs=Image
     imageType=NoImage
     mirroring=NoMirror
     resourceName=None
     resourceObject=/Game/UI/Foundation/Textures/Icons/Items/T-Icon-S-Loot-Wads_01-L
     tiling=NoTile
     imageSize=<128 128>
     margin=<ltrb 0 0 0 0>
     outlineSettings {
      bUseBrushTransparency=false
      roundingType=HalfHeightRadius
      width=0
      color {colorUseRule=UseColor_Specified, specifiedColor=<rgba 0 0 0 0>}
      cornerRadii=<0 0 0 0>
     }
     tintColor {colorUseRule=UseColor_Specified, specifiedColor=<rgba 1 1 1 1>}
     uVRegion {bIsValid=false, max=<0 0>, min=<0 0>}
    }}
   slot {class=OverlaySlot}
  }
  Image_Glow {
   widgetClassPath=&Image
   bIsVariable=true
   parent=Overlay_Reward
   widget=Image_Glow
   properties {
    brush {
     bIsDynamicallyLoaded=false
     drawAs=Image
     imageType=NoImage
     mirroring=NoMirror
     resourceName=None
     resourceObject=/Game/Athena/HUD/LTM/Wax/Materials/M_UI_TextGlow_Add_Gradient_Gold
     tiling=NoTile
     imageSize=<256 256>
     margin=<ltrb 0 0 0 0>
     outlineSettings {
      bUseBrushTransparency=false
      roundingType=HalfHeightRadius
      width=0
      color {colorUseRule=UseColor_Specified, specifiedColor=<rgba 0 0 0 0>}
      cornerRadii=<0 0 0 0>
     }
     tintColor {colorUseRule=UseColor_Specified, specifiedColor=<rgba 1 1 1 1>}
     uVRegion {bIsValid=false, max=<0 0>, min=<0 0>}
    }
    colorAndOpacity=<rgba 1 0.1531 0.2228 1>
   }
   slot {class=OverlaySlot, properties {horizontalAlignment=HAlign_Center, verticalAlignment=VAlign_Center, padding=<ltrb -128 -128 -128 -128>}}
  }
  CommonBorder_0 {
   widgetClassPath=&CommonBorder
   parent=HorizontalBox_1
   widget=CommonBorder_0
   properties {clipping=ClipToBounds, horizontalAlignment=HAlign_Left, padding=<ltrb 0 0 0 0>}
   slot {class=HorizontalBoxSlot, properties {horizontalAlignment=HAlign_Left, verticalAlignment=VAlign_Top}}
  }
  VerticalBox_0 {
   widgetClassPath=&VerticalBox
   parent=CommonBorder_0
   widget=VerticalBox_0
   slot {class=BorderSlot, properties {horizontalAlignment=HAlign_Left, padding=<ltrb 0 0 0 0>}}
  }
  RichText_Title {
   widgetClassPath=&CommonRichTextBlock
   bInherited=true
   bIsVariable=true
   parent=VerticalBox_0
   widget=RichText_Title
   properties {
    bOverrideDefaultStyle=true
    justification=Center
    lineHeightPercentage=0.8
    text=Duel Started
    textTransformPolicy=ToUpper
    wrapTextAt=512
    defaultTextStyleOverride {
     overflowPolicy=Clip
     transformPolicy=None
     colorAndOpacity {colorUseRule=UseColor_Specified, specifiedColor=<rgba 1 0.4564 0 1>}
     font {
      bForceMonospaced=false
      bMaterialIsStencil=false
      fontMaterial=/Game/Athena/HUD/Discoverability/materials/M_UI_DiscoveryText_Inst
      fontObject=&Burbank
      letterSpacing=0
      monospacedWidth=1
      size=40
      skewAmount=0
      typefaceFontName=BurbankRegular-Black
      outlineSettings {
       bApplyOutlineToDropShadows=false
       bMiteredCorners=false
       bSeparateFillAlpha=false
       outlineMaterial=null
       outlineSize=0
       outlineColor=<rgba 0 0 0 1>
      }
     }
     highlightColor {colorUseRule=UseColor_Specified, specifiedColor=<rgba 0 0 0 1>}
     highlightShape {
      bIsDynamicallyLoaded=false
      drawAs=Image
      imageType=NoImage
      mirroring=NoMirror
      resourceName=None
      resourceObject=null
      tiling=NoTile
      imageSize=<32 32>
      margin=<ltrb 0 0 0 0>
      outlineSettings {
       bUseBrushTransparency=false
       roundingType=HalfHeightRadius
       width=0
       color {colorUseRule=UseColor_Specified, specifiedColor=<rgba 0 0 0 0>}
       cornerRadii=<0 0 0 0>
      }
      tintColor {colorUseRule=UseColor_Specified, specifiedColor=<rgba 1 1 1 1>}
      uVRegion {bIsValid=false, max=<0 0>, min=<0 0>}
     }
     selectedBackgroundColor {colorUseRule=UseColor_Foreground, specifiedColor=<rgba 1 0 1 1>}
     shadowColorAndOpacity=<rgba 0 0 0 0.5>
     shadowOffset=<5 5>
     strikeBrush {
      bIsDynamicallyLoaded=false
      drawAs=Image
      imageType=NoImage
      mirroring=NoMirror
      resourceName=None
      resourceObject=null
      tiling=NoTile
      imageSize=<32 32>
      margin=<ltrb 0 0 0 0>
      outlineSettings {
       bUseBrushTransparency=false
       roundingType=HalfHeightRadius
       width=0
       color {colorUseRule=UseColor_Specified, specifiedColor=<rgba 0 0 0 0>}
       cornerRadii=<0 0 0 0>
      }
      tintColor {colorUseRule=UseColor_Specified, specifiedColor=<rgba 1 1 1 1>}
      uVRegion {bIsValid=false, max=<0 0>, min=<0 0>}
     }
     underlineBrush {
      bIsDynamicallyLoaded=false
      drawAs=Image
      imageType=NoImage
      mirroring=NoMirror
      resourceName=None
      resourceObject=null
      tiling=NoTile
      imageSize=<32 32>
      margin=<ltrb 0 0 0 0>
      outlineSettings {
       bUseBrushTransparency=false
       roundingType=HalfHeightRadius
       width=0
       color {colorUseRule=UseColor_Specified, specifiedColor=<rgba 0 0 0 0>}
       cornerRadii=<0 0 0 0>
      }
      tintColor {colorUseRule=UseColor_Specified, specifiedColor=<rgba 1 1 1 1>}
      uVRegion {bIsValid=false, max=<0 0>, min=<0 0>}
     }
    }
    renderTransform {angle=0, scale=<1 1>, shear=<-12 0>, translation=<0 0>}
   }
   slot {class=VerticalBoxSlot, properties {horizontalAlignment=HAlign_Left}}
  }
  RichText_SubTitle {
   widgetClassPath=&CommonRichTextBlock
   bInherited=true
   bIsVariable=true
   parent=VerticalBox_0
   widget=RichText_SubTitle
   properties {
    bOverrideDefaultStyle=true
    justification=Center
    lineHeightPercentage=0.8
    text=Player vs Kit
    textTransformPolicy=ToUpper
    wrapTextAt=1024
    defaultTextStyleOverride {
     overflowPolicy=Clip
     transformPolicy=None
     colorAndOpacity {colorUseRule=UseColor_Specified, specifiedColor=<rgba 1 1 1 1>}
     font {
      bForceMonospaced=false
      bMaterialIsStencil=false
      fontMaterial=&M_UI_DiscoveryText_Inst
      fontObject=&Burbank
      letterSpacing=0
      monospacedWidth=1
      size=24
      skewAmount=0
      typefaceFontName=BurbankRegular-Black
      outlineSettings {
       bApplyOutlineToDropShadows=false
       bMiteredCorners=false
       bSeparateFillAlpha=false
       outlineMaterial=/Game/Athena/HUD/Materials/M_UI_OutlineShine_White_Golden_Inst
       outlineSize=1
       outlineColor=<rgba 1 1 1 1>
      }
     }
     highlightColor {colorUseRule=UseColor_Specified, specifiedColor=<rgba 0 0 0 1>}
     highlightShape {
      bIsDynamicallyLoaded=false
      drawAs=Image
      imageType=NoImage
      mirroring=NoMirror
      resourceName=None
      resourceObject=null
      tiling=NoTile
      imageSize=<32 32>
      margin=<ltrb 0 0 0 0>
      outlineSettings {
       bUseBrushTransparency=false
       roundingType=HalfHeightRadius
       width=0
       color {colorUseRule=UseColor_Specified, specifiedColor=<rgba 0 0 0 0>}
       cornerRadii=<0 0 0 0>
      }
      tintColor {colorUseRule=UseColor_Specified, specifiedColor=<rgba 1 1 1 1>}
      uVRegion {bIsValid=false, max=<0 0>, min=<0 0>}
     }
     selectedBackgroundColor {colorUseRule=UseColor_Foreground, specifiedColor=<rgba 1 0 1 1>}
     shadowColorAndOpacity=<rgba 0 0 0 0.5>
     shadowOffset=<5 5>
     strikeBrush {
      bIsDynamicallyLoaded=false
      drawAs=Image
      imageType=NoImage
      mirroring=NoMirror
      resourceName=None
      resourceObject=null
      tiling=NoTile
      imageSize=<32 32>
      margin=<ltrb 0 0 0 0>
      outlineSettings {
       bUseBrushTransparency=false
       roundingType=HalfHeightRadius
       width=0
       color {colorUseRule=UseColor_Specified, specifiedColor=<rgba 0 0 0 0>}
       cornerRadii=<0 0 0 0>
      }
      tintColor {colorUseRule=UseColor_Specified, specifiedColor=<rgba 1 1 1 1>}
      uVRegion {bIsValid=false, max=<0 0>, min=<0 0>}
     }
     underlineBrush {
      bIsDynamicallyLoaded=false
      drawAs=Image
      imageType=NoImage
      mirroring=NoMirror
      resourceName=None
      resourceObject=null
      tiling=NoTile
      imageSize=<32 32>
      margin=<ltrb 0 0 0 0>
      outlineSettings {
       bUseBrushTransparency=false
       roundingType=HalfHeightRadius
       width=0
       color {colorUseRule=UseColor_Specified, specifiedColor=<rgba 0 0 0 0>}
       cornerRadii=<0 0 0 0>
      }
      tintColor {colorUseRule=UseColor_Specified, specifiedColor=<rgba 1 1 1 1>}
      uVRegion {bIsValid=false, max=<0 0>, min=<0 0>}
     }
    }
    renderTransform {angle=0, scale=<1 1>, shear=<-12 0>, translation=<0 0>}
   }
   slot {class=VerticalBoxSlot, properties {horizontalAlignment=HAlign_Left, verticalAlignment=VAlign_Bottom}}
  }
  Spacer_AnimationPadding {widgetClassPath=&Spacer, parent=MainBox, widget=Spacer_AnimationPadding, properties {size=<1 80>}, slot {class=VerticalBoxSlot}}
  TutorialHighlight {widgetClassPath=&TutorialHighlight}
 ]
}
